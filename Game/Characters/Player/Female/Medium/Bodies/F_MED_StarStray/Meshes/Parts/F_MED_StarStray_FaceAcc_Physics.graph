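# Physics asset for the StarStray female medium character's FaceAcc mesh part, set up for dynamic hair in back, left and right chains.
# Capsules cover the torso, neck, head, upper arms and hair bones, with free angular constraints linking body bones and each chain's master bone to the head.
# Hair segments have per-joint swing and twist limits.
# Extra free angular constraints cross-link mid-lower side segments to the back chain.
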PhysicsAsset /Game/Characters/Player/Female/Medium/Bodies/F_MED_StarStray/Meshes/Parts/F_MED_StarStray_FaceAcc_Physics
bodies {
 clavicle_l [{shapeType=Capsule, length=3.25, radius=5.625, shapeName=clavicle_l_capsule, center=<6.77 -0.5 -1>, rotation=<rot 86 0 0>}]
 clavicle_r [{
   shapeType=Capsule
   length=4.5
   radius=5.625
   shapeName=clavicle_r_capsule
   center=<-5.2595 0.5702 0.8944>
   rotation=<rot 86 -0.0023 177.3418>
  }]
 dyn_hair_bk_00 [{shapeType=Capsule, length=3.75, radius=6.5, shapeName=dyn_hair_bk_00_capsule, center=<5 0 4>, rotation=<rot 0 0 90>}]
 dyn_hair_bk_01 [{shapeType=Capsule, length=3.25, radius=7, shapeName=dyn_hair_bk_01_capsule, center=<6 0 4>, rotation=<rot 0 0 90>}]
 dyn_hair_bk_02 [{shapeType=Capsule, length=3.625, radius=6.5, shapeName=dyn_hair_bk_02_capsule, center=<6 0 4>, rotation=<rot 0 0 90>}]
 dyn_hair_bk_03 [{shapeType=Capsule, length=6.4, radius=5.4, shapeName=dyn_hair_bk_03_capsule, center=<6 0 2.4>, rotation=<rot 0 0 90>}]
 dyn_hair_bk_04 [{shapeType=Capsule, length=5.87, radius=5.75, shapeName=dyn_hair_bk_04_capsule, center=<5.0802 0 2.8>, rotation=<rot 0 0 90>}]
 dyn_hair_bk_05 [{shapeType=Capsule, length=11.75, radius=8, shapeName=dyn_hair_bk_05_capsule, center=<8.8056 0 2.5>, rotation=<rot 0 0 90>}]
 dyn_hair_bk_MSTR [{shapeType=Capsule, length=0.5, radius=0.5, shapeName=dyn_hair_bk_MSTR_capsule}]
 dyn_hair_l_00 [{shapeType=Capsule, length=3.4, radius=3.7, shapeName=dyn_hair_l_00_capsule, center=<3 1 3>, rotation=<rot 90 0 0>}]
 dyn_hair_l_01 [{shapeType=Capsule, length=2.65, radius=4.45, shapeName=dyn_hair_l_01_capsule, center=<4.2972 1.1 -0.25>, rotation=<rot 90 0 0>}]
 dyn_hair_l_02 [{shapeType=Capsule, length=4.5, radius=4.5, shapeName=dyn_hair_l_02_capsule, center=<3.8544 0 2>, rotation=<rot 0 0 90>}]
 dyn_hair_l_03 [{shapeType=Capsule, length=6.35, radius=4.5, shapeName=dyn_hair_l_03_capsule, center=<3.2403 0 3.6>, rotation=<rot 0 0 90>}]
 dyn_hair_l_04 [{shapeType=Capsule, length=5.6, radius=4.6, shapeName=dyn_hair_l_04_capsule, center=<4.7639 0 2.8>, rotation=<rot 0 0 85>}]
 dyn_hair_l_05 [{shapeType=Capsule, length=4.7, radius=5.6, shapeName=dyn_hair_l_05_capsule, center=<5 0 2>, rotation=<rot 0 0 90>}]
 dyn_hair_l_MSTR [{shapeType=Capsule, length=0.5, radius=0.5, shapeName=dyn_hair_l_MSTR_capsule}]
 dyn_hair_r_00 [{shapeType=Capsule, length=3.4, radius=3.7, shapeName=dyn_hair_r_00_capsule, center=<3.3999 -1.626 1>, rotation=<rot 90 90 79.2595>}]
 dyn_hair_r_01 [{shapeType=Capsule, length=2.1, radius=4.4, shapeName=dyn_hair_r_01_capsule, center=<5 0 3>, rotation=<rot 85 180 180>}]
 dyn_hair_r_02 [{shapeType=Capsule, length=3.15, radius=4.6, shapeName=dyn_hair_r_02_capsule, center=<3.9375 2.463 0.5805>, rotation=<rot 0 0 95>}]
 dyn_hair_r_03 [{shapeType=Capsule, length=3.1, radius=4.75, shapeName=dyn_hair_r_03_capsule, center=<1.5578 2.0331 1.6955>, rotation=<rot 0 0 90>}]
 dyn_hair_r_04 [{shapeType=Capsule, length=5.625, radius=4.5, shapeName=dyn_hair_r_04_capsule, center=<3.6297 -0.021 2.3>, rotation=<rot -0.0001 0 78.5833>}]
 dyn_hair_r_05 [{shapeType=Capsule, length=5.7, radius=6.6, shapeName=dyn_hair_r_05_capsule, center=<5 2.2576 3.8025>, rotation=<rot 0 0 95>}]
 dyn_hair_r_MSTR [{shapeType=Capsule, length=0.5, radius=0.5, shapeName=dyn_hair_r_MSTR_capsule}]
 head [{shapeType=Capsule, length=6.1, radius=6.5, shapeName=head_capsule, center=<2.6 2.5 0>, rotation=<rot 90 0 0>}]
 neck_01 [{shapeType=Capsule, length=6.2, radius=4, shapeName=neck_01_capsule, center=<2.6 2 0>, rotation=<rot 90 0 0>}]
 pelvis [{shapeType=Capsule, length=6, radius=14, shapeName=pelvis_capsule, center=<-4.5838 1.0083 0>}]
 spine_02 [{shapeType=Capsule, length=3, radius=14, shapeName=spine_02_capsule, center=<0 6 0>}]
 spine_04 [{shapeType=Capsule, length=4.5, radius=12, shapeName=spine_04_capsule, center=<1 4.1444 0>}]
 spine_05 [{shapeType=Capsule, length=4.25, radius=12, shapeName=spine_05_capsule, center=<1 3.6444 0>}]
 upperarm_l [{shapeType=Capsule, length=19, radius=5.1, shapeName=upperarm_l_capsule, center=<13.8 0 -0.5>, rotation=<rot 90 0 0>}]
 upperarm_r [{shapeType=Capsule, length=19, radius=5.1, shapeName=upperarm_r_capsule, center=<-13 0 0.5>, rotation=<rot 90 0 180>}]
}
constraints [
 {
  bone1Name=spine_02
  bone2Name=pelvis
  swing1LimitDegrees=45
  swing1Motion=Free
  swing2LimitDegrees=45
  swing2Motion=Free
  twistLimitDegrees=45
  twistMotion=Free
 }
 {
  bone1Name=spine_04
  bone2Name=spine_02
  swing1LimitDegrees=45
  swing1Motion=Free
  swing2LimitDegrees=45
  swing2Motion=Free
  twistLimitDegrees=45
  twistMotion=Free
 }
 {
  bone1Name=spine_05
  bone2Name=spine_04
  swing1LimitDegrees=45
  swing1Motion=Free
  swing2LimitDegrees=45
  swing2Motion=Free
  twistLimitDegrees=45
  twistMotion=Free
 }
 {
  bone1Name=clavicle_l
  bone2Name=spine_05
  swing1LimitDegrees=45
  swing1Motion=Free
  swing2LimitDegrees=45
  swing2Motion=Free
  twistLimitDegrees=45
  twistMotion=Free
 }
 {
  bone1Name=clavicle_r
  bone2Name=spine_05
  swing1LimitDegrees=45
  swing1Motion=Free
  swing2LimitDegrees=45
  swing2Motion=Free
  twistLimitDegrees=45
  twistMotion=Free
 }
 {
  bone1Name=upperarm_r
  bone2Name=clavicle_r
  swing1LimitDegrees=45
  swing1Motion=Free
  swing2LimitDegrees=45
  swing2Motion=Free
  twistLimitDegrees=45
  twistMotion=Free
 }
 {
  bone1Name=upperarm_l
  bone2Name=clavicle_l
  swing1LimitDegrees=45
  swing1Motion=Free
  swing2LimitDegrees=45
  swing2Motion=Free
  twistLimitDegrees=45
  twistMotion=Free
 }
 {
  bone1Name=neck_01
  bone2Name=spine_05
  swing1LimitDegrees=45
  swing1Motion=Free
  swing2LimitDegrees=45
  swing2Motion=Free
  twistLimitDegrees=45
  twistMotion=Free
 }
 {
  bone1Name=head
  bone2Name=neck_01
  swing1LimitDegrees=45
  swing1Motion=Free
  swing2LimitDegrees=45
  swing2Motion=Free
  twistLimitDegrees=45
  twistMotion=Free
 }
 {
  bone1Name=dyn_hair_bk_MSTR
  bone2Name=head
  swing1LimitDegrees=45
  swing1Motion=Free
  swing2LimitDegrees=45
  swing2Motion=Free
  twistLimitDegrees=45
  twistMotion=Free
 }
 {
  bone1Name=dyn_hair_l_MSTR
  bone2Name=head
  swing1LimitDegrees=45
  swing1Motion=Free
  swing2LimitDegrees=45
  swing2Motion=Free
  twistLimitDegrees=45
  twistMotion=Free
 }
 {
  bone1Name=dyn_hair_r_MSTR
  bone2Name=head
  swing1LimitDegrees=45
  swing1Motion=Free
  swing2LimitDegrees=45
  swing2Motion=Free
  twistLimitDegrees=45
  twistMotion=Free
 }
 {bone1Name=dyn_hair_bk_00, bone2Name=dyn_hair_bk_MSTR, swing1LimitDegrees=10, swing2LimitDegrees=10, twistLimitDegrees=2}
 {bone1Name=dyn_hair_bk_01, bone2Name=dyn_hair_bk_00, swing1LimitDegrees=20, swing2LimitDegrees=10, twistLimitDegrees=2}
 {bone1Name=dyn_hair_bk_02, bone2Name=dyn_hair_bk_01, swing1LimitDegrees=40, swing2LimitDegrees=40, twistLimitDegrees=50}
 {bone1Name=dyn_hair_bk_03, bone2Name=dyn_hair_bk_02, swing1LimitDegrees=10, swing2LimitDegrees=45, twistLimitDegrees=15}
 {bone1Name=dyn_hair_bk_04, bone2Name=dyn_hair_bk_03, swing1LimitDegrees=5, swing2LimitDegrees=45, twistLimitDegrees=15}
 {bone1Name=dyn_hair_bk_05, bone2Name=dyn_hair_bk_04, swing1LimitDegrees=10, swing2LimitDegrees=45, twistLimitDegrees=2}
 {bone1Name=dyn_hair_l_00, bone2Name=dyn_hair_l_MSTR, swing1LimitDegrees=10, swing2LimitDegrees=20, twistLimitDegrees=2}
 {bone1Name=dyn_hair_l_01, bone2Name=dyn_hair_l_00, swing1LimitDegrees=30, swing2LimitDegrees=10, twistLimitDegrees=2}
 {bone1Name=dyn_hair_l_02, bone2Name=dyn_hair_l_01, swing1LimitDegrees=10, swing2LimitDegrees=25, twistLimitDegrees=2}
 {bone1Name=dyn_hair_l_03, bone2Name=dyn_hair_l_02, swing1LimitDegrees=2, swing2LimitDegrees=45, twistLimitDegrees=2}
 {bone1Name=dyn_hair_l_04, bone2Name=dyn_hair_l_03, swing1LimitDegrees=1, swing2LimitDegrees=30, twistLimitDegrees=2}
 {bone1Name=dyn_hair_l_05, bone2Name=dyn_hair_l_04, swing1LimitDegrees=10, swing2LimitDegrees=30, twistLimitDegrees=2}
 {bone1Name=dyn_hair_r_00, bone2Name=dyn_hair_r_MSTR, swing1LimitDegrees=10, swing2LimitDegrees=20, twistLimitDegrees=2}
 {bone1Name=dyn_hair_r_01, bone2Name=dyn_hair_r_00, swing1LimitDegrees=30, swing2LimitDegrees=10, twistLimitDegrees=2}
 {bone1Name=dyn_hair_r_02, bone2Name=dyn_hair_r_01, swing1LimitDegrees=10, swing2LimitDegrees=25, twistLimitDegrees=2}
 {bone1Name=dyn_hair_r_03, bone2Name=dyn_hair_r_02, swing1LimitDegrees=2, swing2LimitDegrees=45, twistLimitDegrees=2}
 {bone1Name=dyn_hair_r_04, bone2Name=dyn_hair_r_03, swing1LimitDegrees=1, swing2LimitDegrees=30, twistLimitDegrees=2}
 {bone1Name=dyn_hair_r_05, bone2Name=dyn_hair_r_04, swing1LimitDegrees=10, swing2LimitDegrees=30, twistLimitDegrees=2}
 {
  bone1Name=dyn_hair_l_03
  bone2Name=dyn_hair_bk_02
  swing1LimitDegrees=45
  swing1Motion=Free
  swing2LimitDegrees=45
  swing2Motion=Free
  twistLimitDegrees=45
  twistMotion=Free
 }
 {
  bone1Name=dyn_hair_r_03
  bone2Name=dyn_hair_bk_02
  swing1LimitDegrees=45
  swing1Motion=Free
  swing2LimitDegrees=45
  swing2Motion=Free
  twistLimitDegrees=45
  twistMotion=Free
 }
 {
  bone1Name=dyn_hair_l_04
  bone2Name=dyn_hair_bk_03
  swing1LimitDegrees=45
  swing1Motion=Free
  swing2LimitDegrees=45
  swing2Motion=Free
  twistLimitDegrees=45
  twistMotion=Free
 }
 {
  bone1Name=dyn_hair_r_04
  bone2Name=dyn_hair_bk_03
  swing1LimitDegrees=45
  swing1Motion=Free
  swing2LimitDegrees=45
  swing2Motion=Free
  twistLimitDegrees=45
  twistMotion=Free
 }
]
