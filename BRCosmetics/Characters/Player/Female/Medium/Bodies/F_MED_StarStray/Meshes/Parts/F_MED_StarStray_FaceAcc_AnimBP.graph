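# Post-copy AnimBP for the StarStray female skin's FaceAcc mesh, driving its dyn_hair bone chains from the parent mesh's pose.
# RigidBody sims swap physics assets and base bones by backpack and skydive, jump or fall state, blending to unsimulated PreSimPose when rigid body is disabled.
# Constraints pin hair bones to head or spine_05, while stacked additive bone offsets, some weighted by head rotation or toggled by flight state, shape the hair.
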AnimBlueprint /BRCosmetics/Characters/Player/Female/Medium/Bodies/F_MED_StarStray/Meshes/Parts/F_MED_StarStray_FaceAcc_AnimBP
target_skeleton=/Game/Characters/Player/Female/Medium/Bodies/F_MED_StarStray/Meshes/Parts/F_MED_StarStray_FaceAcc_Skeleton
graphs [
 AnimGraph {nodes [
   n1 = "Constraint - dyn_hair_bk_00 (P)"(ComponentPose=Local To Component(LocalPose=Use cached pose 'StartPose'()), Alpha=1, ConstraintWeights_0=1) {
    alphaBoolBlend {blendOption=Linear}
    boneToModify {boneName=dyn_hair_bk_00}
    constraintSetup [{targetBone {boneName=head}}]
    constraintWeights [1]
   }
   n2 = "Constraint - dyn_hair_l_00 (P)"(ComponentPose=@n1, Alpha=1, ConstraintWeights_0=1) {
    alphaBoolBlend {blendOption=Linear}
    boneToModify {boneName=dyn_hair_l_00}
    constraintSetup [{targetBone {boneName=head}}]
    constraintWeights [1]
   }
   n3 = "Constraint - dyn_hair_r_00 (P)"(ComponentPose=@n2, Alpha=1, ConstraintWeights_0=1) {
    alphaBoolBlend {blendOption=Linear}
    boneToModify {boneName=dyn_hair_r_00}
    constraintSetup [{targetBone {boneName=head}}]
    constraintWeights [1]
   }
   n4 = "Constraint - dyn_hair_bk_01 (P)"(ComponentPose=@n3, Alpha=1, ConstraintWeights_0=1) {
    alphaBoolBlend {blendOption=Linear}
    boneToModify {boneName=dyn_hair_bk_01}
    constraintSetup [{targetBone {boneName=head}}]
    constraintWeights [1]
   }
   n5 = "Constraint - dyn_hair_l_01 (P)"(ComponentPose=@n4, Alpha=1, ConstraintWeights_0=1) {
    alphaBoolBlend {blendOption=Linear}
    boneToModify {boneName=dyn_hair_l_01}
    constraintSetup [{targetBone {boneName=head}}]
    constraintWeights [1]
   }
   n6 = "Constraint - dyn_hair_r_01 (P)"(ComponentPose=@n5, Alpha=1, ConstraintWeights_0=1) {
    alphaBoolBlend {blendOption=Linear}
    boneToModify {boneName=dyn_hair_r_01}
    constraintSetup [{targetBone {boneName=head}}]
    constraintWeights [1]
   }
   n7 = "Constraint - dyn_hair_bk_02 (P)"(ComponentPose=@n6, Alpha=1, ConstraintWeights_0=1) {
    alphaBoolBlend {blendOption=Linear}
    boneToModify {boneName=dyn_hair_bk_02}
    constraintSetup [{targetBone {boneName=spine_05}}]
    constraintWeights [1]
   }
   n8 = "Constraint - dyn_hair_l_02 (P)"(ComponentPose=@n7, Alpha=1, ConstraintWeights_0=1) {
    alphaBoolBlend {blendOption=Linear}
    boneToModify {boneName=dyn_hair_l_02}
    constraintSetup [{targetBone {boneName=spine_05}}]
    constraintWeights [1]
   }
   n9 = "Constraint - dyn_hair_r_02 (P)"(ComponentPose=@n8, Alpha=1, ConstraintWeights_0=1) {
    alphaBoolBlend {blendOption=Linear}
    boneToModify {boneName=dyn_hair_r_02}
    constraintSetup [{targetBone {boneName=spine_05}}]
    constraintWeights [1]
   }
   n10 = RigidBody!(
    ComponentPose=@n9
    Alpha
    OverrideWorldGravity:Vector=Get GravityOverride()
    ExternalForce
    ComponentLinearAccScale:Vector=Get AccScaleOverride() -> (AccScaleOverride:Vector)
   ) -> (Pose) {
    bClampLinearTranslationLimitToRefPose=true
    bOverrideWorldGravity=true
    overridePhysicsAsset=&/F_MED_StarStray_FaceAcc_Physics_InAir
    simulationSpace=BaseBoneSpace
    alphaBoolBlend {blendOption=Linear}
    alphaScaleBiasClamp {interpSpeedDecreasing=4, interpSpeedIncreasing=6}
    baseBoneRef {boneName=neck_01}
    componentAppliedLinearAccClamp=<800 800 500>
    componentLinearAccScale=<0.6 0.6 0.2>
    componentLinearVelScale=<0.5 0.5 0.5>
    overrideWorldGravity {z=-875}
    simSpaceSettings {maxAngularAcceleration=200, maxAngularVelocity=5, maxLinearAcceleration=500, maxLinearVelocity=200}
   }
   n11 = "Transform (Modify) Bone - Bone: dyn_hair_l_00"!(ComponentPose=@n10, bAlphaBoolEnabled:bool=True, Translation="0,0,0", Rotation="50,3,0", Scale="1,1,1") -> (Pose) {
    alphaInputType=Bool
    rotationMode=BMM_Additive
    rotationSpace=BCS_BoneSpace
    translationMode=BMM_Additive
    translationSpace=BCS_BoneSpace
    alphaBoolBlend {blendOption=Linear}
    boneToModify {boneName=dyn_hair_l_00}
    rotation {pitch=50, yaw=3}
   }
   n12 = "Transform (Modify) Bone - Bone: dyn_hair_bk_00"!(ComponentPose=@n11, bAlphaBoolEnabled:bool=True, Translation="0,0,0", Rotation="40,0,0", Scale="1,1,1") -> (Pose) {
    alphaInputType=Bool
    rotationMode=BMM_Additive
    rotationSpace=BCS_BoneSpace
    translationMode=BMM_Additive
    translationSpace=BCS_BoneSpace
    alphaBoolBlend {blendOption=Linear}
    boneToModify {boneName=dyn_hair_bk_00}
    rotation {pitch=40}
   }
   n13 = "Transform (Modify) Bone - Bone: dyn_hair_r_00"!(ComponentPose=@n12, bAlphaBoolEnabled:bool=True, Translation="0,0,0", Rotation="40,-3,0", Scale="1,1,1") -> (Pose) {
    alphaInputType=Bool
    rotationMode=BMM_Additive
    rotationSpace=BCS_BoneSpace
    translationMode=BMM_Additive
    translationSpace=BCS_BoneSpace
    alphaBoolBlend {blendOption=Linear}
    boneToModify {boneName=dyn_hair_r_00}
    rotation {pitch=40, yaw=-3}
   }
   n14 = "Transform (Modify) Bone - Bone: dyn_hair_l_01"!(ComponentPose=@n13, bAlphaBoolEnabled:bool=True, Translation="0,0,0", Rotation="10,5,0", Scale="1,1,1") -> (Pose) {
    alphaInputType=Bool
    rotationMode=BMM_Additive
    rotationSpace=BCS_BoneSpace
    translationMode=BMM_Additive
    translationSpace=BCS_BoneSpace
    alphaBoolBlend {blendOption=Linear}
    boneToModify {boneName=dyn_hair_l_01}
    rotation {pitch=10, yaw=5}
   }
   n15 = "Transform (Modify) Bone - Bone: dyn_hair_bk_01"!(ComponentPose=@n14, bAlphaBoolEnabled:bool=True, Translation="0,0,0", Rotation="20,0,0", Scale="1,1,1") -> (Pose) {
    alphaInputType=Bool
    rotationMode=BMM_Additive
    rotationSpace=BCS_BoneSpace
    translationMode=BMM_Additive
    translationSpace=BCS_BoneSpace
    alphaBoolBlend {blendOption=Linear}
    boneToModify {boneName=dyn_hair_bk_01}
    rotation {pitch=20}
   }
   n16 = "Transform (Modify) Bone - Bone: dyn_hair_r_01"!(ComponentPose=@n15, bAlphaBoolEnabled:bool=True, Translation="0,0,0", Rotation="30,-5,0", Scale="1,1,1") -> (Pose) {
    alphaInputType=Bool
    rotationMode=BMM_Additive
    rotationSpace=BCS_BoneSpace
    translationMode=BMM_Additive
    translationSpace=BCS_BoneSpace
    alphaBoolBlend {blendOption=Linear}
    boneToModify {boneName=dyn_hair_r_01}
    rotation {pitch=30, yaw=-5}
   }
   n17 = "Transform (Modify) Bone - Bone: dyn_hair_l_02"!(ComponentPose=@n16, bAlphaBoolEnabled:bool=True, Translation="0,0,0", Rotation="-90,0,0", Scale="1,1,1") -> (Pose) {
    alphaInputType=Bool
    rotationMode=BMM_Additive
    rotationSpace=BCS_BoneSpace
    translationMode=BMM_Additive
    translationSpace=BCS_BoneSpace
    alphaBoolBlend {blendOption=Linear}
    boneToModify {boneName=dyn_hair_l_02}
    rotation {pitch=-90}
   }
   n18 = "Transform (Modify) Bone - Bone: dyn_hair_bk_02"!(ComponentPose=@n17, bAlphaBoolEnabled:bool=True, Translation="0,0,0", Rotation="-90,0,0", Scale="1,1,1") -> (Pose) {
    alphaInputType=Bool
    rotationMode=BMM_Additive
    rotationSpace=BCS_BoneSpace
    translationMode=BMM_Additive
    translationSpace=BCS_BoneSpace
    alphaBoolBlend {blendOption=Linear}
    boneToModify {boneName=dyn_hair_bk_02}
    rotation {pitch=-90}
   }
   n19 = "Transform (Modify) Bone - Bone: dyn_hair_r_02"!(ComponentPose=@n18, bAlphaBoolEnabled:bool=True, Translation="0,0,0", Rotation="-100,0,0", Scale="1,1,1") -> (Pose) {
    alphaInputType=Bool
    rotationMode=BMM_Additive
    rotationSpace=BCS_BoneSpace
    translationMode=BMM_Additive
    translationSpace=BCS_BoneSpace
    alphaBoolBlend {blendOption=Linear}
    boneToModify {boneName=dyn_hair_r_02}
    rotation {pitch=-100}
   }
   n20 = "Transform (Modify) Bone - Bone: dyn_hair_bk_03"!(ComponentPose=@n19, bAlphaBoolEnabled:bool=True, Translation="0,0,2", Rotation="0,0,0", Scale="1,1,1") -> (Pose) {
    alphaInputType=Bool
    rotationMode=BMM_Additive
    rotationSpace=BCS_BoneSpace
    translationMode=BMM_Additive
    translationSpace=BCS_BoneSpace
    alphaBoolBlend {blendOption=Linear}
    boneToModify {boneName=dyn_hair_bk_03}
    translation {z=2}
   }
   n21 = "Transform (Modify) Bone - Bone: dyn_hair_r_05"!(ComponentPose=@n20, bAlphaBoolEnabled:bool=True, Translation="0,0,0", Rotation="10,0,0", Scale="1,1,1") -> (Pose) {
    alphaInputType=Bool
    rotationMode=BMM_Additive
    rotationSpace=BCS_BoneSpace
    translationMode=BMM_Additive
    translationSpace=BCS_BoneSpace
    alphaBoolBlend {blendOption=Linear}
    boneToModify {boneName=dyn_hair_r_05}
    rotation {pitch=10}
   }
   n22 = Component To Local(ComponentPose=@n21)
   n23 = Use cached pose 'RBPose'()
   n24 = Get IsSkyFall() -> (IsSkyFall:bool)
   n25 = "Transform (Modify) Bone - Bone: dyn_hair_l_MSTR"!(
    ComponentPose=Local To Component(LocalPose=Use cached pose 'RBPose'())
    bAlphaBoolEnabled:bool=True
    Translation="0,0,0"
    Rotation="10,0,0"
    Scale="1,1,1"
   ) -> (Pose) {
    alphaInputType=Bool
    rotationMode=BMM_Additive
    rotationSpace=BCS_BoneSpace
    translationMode=BMM_Additive
    translationSpace=BCS_BoneSpace
    alphaBoolBlend {blendOption=Linear}
    boneToModify {boneName=dyn_hair_l_MSTR}
    rotation {pitch=10}
   }
   n26 = "Transform (Modify) Bone - Bone: dyn_hair_l_02"(ComponentPose=@n25, bAlphaBoolEnabled=True, Translation="0,0,0", Rotation="-20,0,0", Scale="1,1,1") {
    alphaInputType=Bool
    rotationMode=BMM_Additive
    rotationSpace=BCS_BoneSpace
    translationMode=BMM_Additive
    translationSpace=BCS_BoneSpace
    alphaBoolBlend {blendOption=Linear}
    boneToModify {boneName=dyn_hair_l_02}
    rotation {pitch=-20}
   }
   n27 = "Transform (Modify) Bone - Bone: dyn_hair_l_03"!(ComponentPose=@n26, bAlphaBoolEnabled:bool=True, Translation="0,0,0", Rotation="30,0,0", Scale="1,1,1") -> (Pose) {
    alphaInputType=Bool
    rotationMode=BMM_Additive
    rotationSpace=BCS_BoneSpace
    translationMode=BMM_Additive
    translationSpace=BCS_BoneSpace
    alphaBoolBlend {blendOption=Linear}
    boneToModify {boneName=dyn_hair_l_03}
    rotation {pitch=30}
   }
   n28 = "Transform (Modify) Bone - Bone: dyn_hair_l_04"!(ComponentPose=@n27, bAlphaBoolEnabled:bool=True, Translation="0,0,0", Rotation="20,0,0", Scale="1,1,1") -> (Pose) {
    alphaInputType=Bool
    rotationMode=BMM_Additive
    rotationSpace=BCS_BoneSpace
    translationMode=BMM_Additive
    translationSpace=BCS_BoneSpace
    alphaBoolBlend {blendOption=Linear}
    boneToModify {boneName=dyn_hair_l_04}
    rotation {pitch=20}
   }
   n29 = "Transform (Modify) Bone - Bone: dyn_hair_bk_03"(ComponentPose=@n28, bAlphaBoolEnabled=True, Translation="0,0,0", Rotation="20,0,0", Scale="1,1,1") {
    alphaInputType=Bool
    rotationMode=BMM_Additive
    rotationSpace=BCS_BoneSpace
    translationMode=BMM_Additive
    translationSpace=BCS_BoneSpace
    alphaBoolBlend {blendOption=Linear}
    boneToModify {boneName=dyn_hair_bk_03}
    rotation {pitch=20}
   }
   n30 = "Transform (Modify) Bone - Bone: dyn_hair_bk_04"!(ComponentPose=@n29, bAlphaBoolEnabled:bool=True, Translation="0,0,0", Rotation="-20,0,0", Scale="1,1,1") -> (Pose) {
    alphaInputType=Bool
    rotationMode=BMM_Additive
    rotationSpace=BCS_BoneSpace
    translationMode=BMM_Additive
    translationSpace=BCS_BoneSpace
    alphaBoolBlend {blendOption=Linear}
    boneToModify {boneName=dyn_hair_bk_04}
    rotation {pitch=-20}
   }
   n31 = "Transform (Modify) Bone - Bone: dyn_hair_r_03"!(ComponentPose=@n30, bAlphaBoolEnabled:bool=True, Translation="0,0,0", Rotation="10,0,0", Scale="1,1,1") -> (Pose) {
    alphaInputType=Bool
    rotationMode=BMM_Additive
    rotationSpace=BCS_BoneSpace
    translationMode=BMM_Additive
    translationSpace=BCS_BoneSpace
    alphaBoolBlend {blendOption=Linear}
    boneToModify {boneName=dyn_hair_r_03}
    rotation {pitch=10}
   }
   n32 = "Transform (Modify) Bone - Bone: dyn_hair_r_04"!(ComponentPose=@n31, bAlphaBoolEnabled:bool=True, Translation="0,0,0", Rotation="0,-10,0", Scale="1,1,1") -> (Pose) {
    alphaInputType=Bool
    rotationMode=BMM_Additive
    rotationSpace=BCS_BoneSpace
    translationMode=BMM_Additive
    translationSpace=BCS_BoneSpace
    alphaBoolBlend {blendOption=Linear}
    boneToModify {boneName=dyn_hair_r_04}
    rotation {yaw=-10}
   }
   n33 = "Transform (Modify) Bone - Bone: dyn_hair_r_01"(ComponentPose=@n32, bAlphaBoolEnabled=True, Translation="0,0,0", Rotation="10,0,0", Scale="1,1,1") {
    alphaInputType=Bool
    rotationMode=BMM_Additive
    rotationSpace=BCS_BoneSpace
    translationMode=BMM_Additive
    translationSpace=BCS_BoneSpace
    alphaBoolBlend {blendOption=Linear}
    boneToModify {boneName=dyn_hair_r_01}
    rotation {pitch=10}
   }
   n34 = RigidBody(
    ComponentPose=Local To Component(LocalPose=Use cached pose 'StartPose'())
    OverrideWorldGravity=Get GravityOverride()
    ComponentLinearAccScale=Get AccScaleOverride()
   ) {
    bClampLinearTranslationLimitToRefPose=true
    bOverrideWorldGravity=true
    overridePhysicsAsset=/Game/Characters/Player/Female/Medium/Bodies/F_MED_StarStray/Meshes/Parts/F_MED_StarStray_FaceAcc_Physics
    simulationSpace=BaseBoneSpace
    alphaBoolBlend {blendOption=Linear}
    alphaScaleBiasClamp {interpSpeedDecreasing=4, interpSpeedIncreasing=6}
    baseBoneRef {boneName=neck_01}
    componentAppliedLinearAccClamp=<800 800 500>
    componentLinearAccScale=<0.6 0.6 0.2>
    componentLinearVelScale=<0.5 0.5 0.5>
    overrideWorldGravity {z=-875}
    simSpaceSettings {maxAngularAcceleration=200, maxAngularVelocity=5, maxLinearAcceleration=500, maxLinearVelocity=200}
   }
   n35 = Blend Poses by bool(
    BlendPose_0=Component To Local(ComponentPose=@n34)
    BlendPose_1=Blend Poses by bool(
     BlendPose_0=Component To Local(ComponentPose=@n33)
     BlendPose_1=Blend Poses by bool(BlendPose_0=@n22, BlendPose_1=@n23, BlendTime_0=1, BlendTime_1=0.3, bActiveValue=@n24) {blendType=Linear}
     BlendTime_0=0.3
     BlendTime_1=0.2
     bActiveValue=Get IsSkyJump() -> (IsSkyJump:bool)
    ) {blendType=Linear}
    BlendTime_0=0.3
    BlendTime_1=0.2
    bActiveValue=Get bIsSkydiveDiveMode()
   ) {blendType=Linear}
   n36 = Get bIsParachuteOpened()
   n37 = "Transform (Modify) Bone - Bone: dyn_hair_l_03"(
    ComponentPose=Local To Component(LocalPose=Use cached pose 'PreSimPose'())
    bAlphaBoolEnabled=@n36
    Translation="0,2,0"
    Rotation="0,0,0"
    Scale="1,1,1"
   ) {
    alphaInputType=Bool
    rotationMode=BMM_Additive
    rotationSpace=BCS_BoneSpace
    translationMode=BMM_Additive
    translationSpace=BCS_BoneSpace
    alphaBoolBlend {blendOption=Linear}
    boneToModify {boneName=dyn_hair_l_03}
    translation {y=2}
   }
   n38 = "Transform (Modify) Bone - Bone: dyn_hair_bk_03"(ComponentPose=@n37, bAlphaBoolEnabled=@n36, Translation="0,2,1", Rotation="0,0,0", Scale="1,1,1") {
    alphaInputType=Bool
    rotationMode=BMM_Additive
    rotationSpace=BCS_BoneSpace
    translationMode=BMM_Additive
    translationSpace=BCS_BoneSpace
    alphaBoolBlend {blendOption=Linear}
    boneToModify {boneName=dyn_hair_bk_03}
    translation {y=2, z=1}
   }
   n39 = "Transform (Modify) Bone - Bone: dyn_hair_r_03"(ComponentPose=@n38, bAlphaBoolEnabled=@n36, Translation="0,2,0", Rotation="0,0,0", Scale="1,1,1") {
    alphaInputType=Bool
    rotationMode=BMM_Additive
    rotationSpace=BCS_BoneSpace
    translationMode=BMM_Additive
    translationSpace=BCS_BoneSpace
    alphaBoolBlend {blendOption=Linear}
    boneToModify {boneName=dyn_hair_r_03}
    translation {y=2}
   }
   n40 = Component To Local(ComponentPose=@n39)
   n41 = "Transform (Modify) Bone - Bone: dyn_hair_l_04"(
    ComponentPose=Local To Component(LocalPose=Use cached pose 'PreSimPose'())
    bAlphaBoolEnabled=True
    Translation="0,0,1"
    Rotation="30,0,0"
    Scale="1,1,1"
   ) {
    alphaInputType=Bool
    rotationMode=BMM_Additive
    rotationSpace=BCS_BoneSpace
    translationMode=BMM_Additive
    translationSpace=BCS_BoneSpace
    alphaBoolBlend {blendOption=Linear}
    boneToModify {boneName=dyn_hair_l_04}
    rotation {pitch=30}
    translation {z=1}
   }
   n42 = "Transform (Modify) Bone - Bone: dyn_hair_bk_04"(ComponentPose=@n41, bAlphaBoolEnabled=True, Translation="0,0,1", Rotation="35,0,0", Scale="1,1,1") {
    alphaInputType=Bool
    rotationMode=BMM_Additive
    rotationSpace=BCS_BoneSpace
    translationMode=BMM_Additive
    translationSpace=BCS_BoneSpace
    alphaBoolBlend {blendOption=Linear}
    boneToModify {boneName=dyn_hair_bk_04}
    rotation {pitch=35}
    translation {z=1}
   }
   n43 = "Transform (Modify) Bone - Bone: dyn_hair_r_04"(ComponentPose=@n42, bAlphaBoolEnabled=True, Translation="0,-1,1", Rotation="30,-10,0", Scale="1,1,1") {
    alphaInputType=Bool
    rotationMode=BMM_Additive
    rotationSpace=BCS_BoneSpace
    translationMode=BMM_Additive
    translationSpace=BCS_BoneSpace
    alphaBoolBlend {blendOption=Linear}
    boneToModify {boneName=dyn_hair_r_04}
    rotation {pitch=30, yaw=-10}
    translation {y=-1, z=1}
   }
   n44 = Get bIsSkydiveDiveMode()
   n45 = "Transform (Modify) Bone - Bone: dyn_hair_l_04"(ComponentPose=@n43, bAlphaBoolEnabled=@n44, Translation="0,0,1", Rotation="-20,0,0", Scale="1,1,1") {
    alphaInputType=Bool
    rotationMode=BMM_Additive
    rotationSpace=BCS_BoneSpace
    translationMode=BMM_Additive
    translationSpace=BCS_BoneSpace
    alphaBoolBlend {blendOption=Linear}
    boneToModify {boneName=dyn_hair_l_04}
    rotation {pitch=-20}
    translation {z=1}
   }
   n46 = "Transform (Modify) Bone - Bone: dyn_hair_bk_04"(ComponentPose=@n45, bAlphaBoolEnabled=@n44, Translation="0,0,0", Rotation="-20,0,0", Scale="1,1,1") {
    alphaInputType=Bool
    rotationMode=BMM_Additive
    rotationSpace=BCS_BoneSpace
    translationMode=BMM_Additive
    translationSpace=BCS_BoneSpace
    alphaBoolBlend {blendOption=Linear}
    boneToModify {boneName=dyn_hair_bk_04}
    rotation {pitch=-20}
   }
   n47 = "Transform (Modify) Bone - Bone: dyn_hair_r_04"(ComponentPose=@n46, bAlphaBoolEnabled=@n44, Translation="0,0,0", Rotation="-20,0,0", Scale="1,1,1") {
    alphaInputType=Bool
    rotationMode=BMM_Additive
    rotationSpace=BCS_BoneSpace
    translationMode=BMM_Additive
    translationSpace=BCS_BoneSpace
    alphaBoolBlend {blendOption=Linear}
    boneToModify {boneName=dyn_hair_r_04}
    rotation {pitch=-20}
   }
   n48 = Component To Local(ComponentPose=@n47)
   n49 = Get IsInSkyWithoutParachute() -> (IsInSkyWithoutParachute:bool)
   n50 = RigidBody(
    ComponentPose=Local To Component(LocalPose=Blend Poses by bool(BlendPose_0=@n48, BlendPose_1=@n40, BlendTime_0=0.5, BlendTime_1=0.5, bActiveValue=@n49) {blendType=Linear})
    OverrideWorldGravity=Get GravityOverride()
    ComponentLinearAccScale=Get AccScaleOverride()
   ) {
    bClampLinearTranslationLimitToRefPose=true
    bOverrideWorldGravity=true
    overridePhysicsAsset=&/F_MED_StarStray_FaceAcc_Physics_withBack
    simulationSpace=BaseBoneSpace
    alphaBoolBlend {blendOption=Linear}
    alphaScaleBiasClamp {interpSpeedDecreasing=4, interpSpeedIncreasing=6}
    baseBoneRef {boneName=spine_05}
    componentAppliedLinearAccClamp=<800 800 500>
    componentLinearAccScale=<0.6 0.6 0.2>
    componentLinearVelScale=<0.5 0.5 0.5>
    overrideWorldGravity {z=-875}
    simSpaceSettings {maxAngularAcceleration=200, maxAngularVelocity=5, maxLinearAcceleration=500, maxLinearVelocity=200}
   }
   n51 = Component To Local(ComponentPose=@n50)
   n52 = Get WearingBackpack() -> (WearingBackpack:bool)
   Output Pose(Result=Blend Poses by bool(
     BlendPose_0=Blend Poses by bool(BlendPose_0=@n51, BlendPose_1=@n35, BlendTime_0=0.2, BlendTime_1=0.2, bActiveValue=@n52) {blendType=Linear}
     BlendPose_1=Use cached pose 'PreSimPose'()
     BlendTime_0=0.2
     BlendTime_1=0.2
     bActiveValue=Get bRigidBodyEnabled()
    ) {blendType=Linear}) {layerGroup=None}
   Save cached pose 'StartPose'(Pose=Copy Pose From Mesh() {bUseAttachedParent=true}) {cacheName=StartPose}
   n53 = Get HeadRorBk() -> (HeadRorBk:double)
   n54 = "Transform (Modify) Bone - Bone: dyn_hair_l_MSTR"!(
    ComponentPose=Local To Component(LocalPose=Use cached pose 'StartPose'())
    Alpha:float=@n53
    Translation="2,-1,0"
    Rotation="70,0,0"
    Scale="1,1,1"
   ) -> (Pose) {
    rotationMode=BMM_Additive
    rotationSpace=BCS_BoneSpace
    translationMode=BMM_Additive
    translationSpace=BCS_BoneSpace
    alphaBoolBlend {blendOption=Linear}
    boneToModify {boneName=dyn_hair_l_MSTR}
    rotation {pitch=70}
    translation=<2 -1>
   }
   n55 = "Transform (Modify) Bone - Bone: dyn_hair_bk_MSTR"(ComponentPose=@n54, Alpha=@n53, Translation="0,0,0", Rotation="50,0,0", Scale="1,1,1") {
    rotationMode=BMM_Additive
    rotationSpace=BCS_BoneSpace
    translationMode=BMM_Additive
    translationSpace=BCS_BoneSpace
    alphaBoolBlend {blendOption=Linear}
    boneToModify {boneName=dyn_hair_bk_MSTR}
    rotation {pitch=50}
   }
   n56 = "Transform (Modify) Bone - Bone: dyn_hair_r_MSTR"(ComponentPose=@n55, Alpha=@n53, Translation="-1,-0.5,5", Rotation="50,0,0", Scale="1,1,1") {
    rotationMode=BMM_Additive
    rotationSpace=BCS_BoneSpace
    translationMode=BMM_Additive
    translationSpace=BCS_BoneSpace
    alphaBoolBlend {blendOption=Linear}
    boneToModify {boneName=dyn_hair_r_MSTR}
    rotation {pitch=50}
    translation=<-1 -0.5 5>
   }
   n57 = Get HeadRorR() -> (HeadRorR:double)
   n58 = "Transform (Modify) Bone - Bone: dyn_hair_l_MSTR"(ComponentPose=@n56, Alpha=@n57, Translation="0,2,3", Rotation="0,-10,0", Scale="1,1,1") {
    rotationMode=BMM_Additive
    rotationSpace=BCS_BoneSpace
    translationMode=BMM_Additive
    translationSpace=BCS_BoneSpace
    alphaBoolBlend {blendOption=Linear}
    boneToModify {boneName=dyn_hair_l_MSTR}
    rotation {yaw=-10}
    translation {y=2, z=3}
   }
   n59 = "Transform (Modify) Bone - Bone: dyn_hair_bk_MSTR"(ComponentPose=@n58, Alpha=@n57, Translation="-2,2,5", Rotation="-20,0,0", Scale="1,1,1") {
    rotationMode=BMM_Additive
    rotationSpace=BCS_BoneSpace
    translationMode=BMM_Additive
    translationSpace=BCS_BoneSpace
    alphaBoolBlend {blendOption=Linear}
    boneToModify {boneName=dyn_hair_bk_MSTR}
    rotation {pitch=-20}
    translation=<-2 2 5>
   }
   n60 = "Transform (Modify) Bone - Bone: dyn_hair_r_MSTR"(ComponentPose=@n59, Alpha=@n57, Translation="-6,0,0", Rotation="0,0,0", Scale="1,1,1") {
    rotationMode=BMM_Additive
    rotationSpace=BCS_BoneSpace
    translationMode=BMM_Additive
    translationSpace=BCS_BoneSpace
    alphaBoolBlend {blendOption=Linear}
    boneToModify {boneName=dyn_hair_r_MSTR}
    translation {x=-6}
   }
   n61 = RigidBody(ComponentPose=@n60, OverrideWorldGravity=Get GravityOverride(), ComponentLinearAccScale=Get AccScaleOverride()) {
    bClampLinearTranslationLimitToRefPose=true
    bOverrideWorldGravity=true
    overridePhysicsAsset=&F_MED_StarStray_FaceAcc_Physics
    simulationSpace=BaseBoneSpace
    alphaBoolBlend {blendOption=Linear}
    alphaScaleBiasClamp {interpSpeedDecreasing=4, interpSpeedIncreasing=6}
    baseBoneRef {boneName=neck_01}
    componentAppliedLinearAccClamp=<800 800 500>
    componentLinearAccScale=<0.6 0.6 0.2>
    componentLinearVelScale=<0.5 0.5 0.5>
    overrideWorldGravity {z=-875}
    simSpaceSettings {maxAngularAcceleration=200, maxAngularVelocity=5, maxLinearAcceleration=500, maxLinearVelocity=200}
   }
   Save cached pose 'RBPose'(Pose=Component To Local(ComponentPose=@n61)) {cacheName=RBPose}
   n62 = "Constraint - dyn_hair_bk_00 (P)"(ComponentPose=Local To Component(LocalPose=Use cached pose 'StartPose'()), Alpha=1, ConstraintWeights_0=1) {
    alphaBoolBlend {blendOption=Linear}
    boneToModify {boneName=dyn_hair_bk_00}
    constraintSetup [{targetBone {boneName=head}}]
    constraintWeights [1]
   }
   n63 = "Constraint - dyn_hair_l_00 (P)"(ComponentPose=@n62, Alpha=1, ConstraintWeights_0=1) {
    alphaBoolBlend {blendOption=Linear}
    boneToModify {boneName=dyn_hair_l_00}
    constraintSetup [{targetBone {boneName=head}}]
    constraintWeights [1]
   }
   n64 = "Constraint - dyn_hair_r_00 (P)"(ComponentPose=@n63, Alpha=1, ConstraintWeights_0=1) {
    alphaBoolBlend {blendOption=Linear}
    boneToModify {boneName=dyn_hair_r_00}
    constraintSetup [{targetBone {boneName=head}}]
    constraintWeights [1]
   }
   n65 = "Transform (Modify) Bone - Bone: dyn_hair_l_00"!(ComponentPose=@n64, Alpha:float=1, Translation="0,0,1", Rotation="40,0,0", Scale="1,1,1") -> (Pose) {
    rotationMode=BMM_Additive
    rotationSpace=BCS_BoneSpace
    translationMode=BMM_Additive
    translationSpace=BCS_BoneSpace
    alphaBoolBlend {blendOption=Linear}
    boneToModify {boneName=dyn_hair_l_00}
    rotation {pitch=40}
    translation {z=1}
   }
   n66 = "Transform (Modify) Bone - Bone: dyn_hair_bk_00"!(ComponentPose=@n65, Alpha:float=1, Translation="0,0,0", Rotation="40,0,0", Scale="1,1,1") -> (Pose) {
    rotationMode=BMM_Additive
    rotationSpace=BCS_BoneSpace
    translationMode=BMM_Additive
    translationSpace=BCS_BoneSpace
    alphaBoolBlend {blendOption=Linear}
    boneToModify {boneName=dyn_hair_bk_00}
    rotation {pitch=40}
   }
   n67 = "Transform (Modify) Bone - Bone: dyn_hair_r_00"!(ComponentPose=@n66, Alpha:float=1, Translation="0,0,5", Rotation="60,0,0", Scale="1,1,1") -> (Pose) {
    rotationMode=BMM_Additive
    rotationSpace=BCS_BoneSpace
    translationMode=BMM_Additive
    translationSpace=BCS_BoneSpace
    alphaBoolBlend {blendOption=Linear}
    boneToModify {boneName=dyn_hair_r_00}
    rotation {pitch=60}
    translation {z=5}
   }
   n68 = "Constraint - dyn_hair_bk_01 (P)"(ComponentPose=@n67, Alpha=1, ConstraintWeights_0=1) {
    alphaBoolBlend {blendOption=Linear}
    boneToModify {boneName=dyn_hair_bk_01}
    constraintSetup [{targetBone {boneName=head}}]
    constraintWeights [1]
   }
   n69 = "Constraint - dyn_hair_l_01 (P)"(ComponentPose=@n68, Alpha=1, ConstraintWeights_0=1) {
    alphaBoolBlend {blendOption=Linear}
    boneToModify {boneName=dyn_hair_l_01}
    constraintSetup [{targetBone {boneName=head}}]
    constraintWeights [1]
   }
   n70 = "Constraint - dyn_hair_r_01 (P)"(ComponentPose=@n69, Alpha=1, ConstraintWeights_0=1) {
    alphaBoolBlend {blendOption=Linear}
    boneToModify {boneName=dyn_hair_r_01}
    constraintSetup [{targetBone {boneName=head}}]
    constraintWeights [1]
   }
   n71 = "Transform (Modify) Bone - Bone: dyn_hair_l_01"!(ComponentPose=@n70, Alpha:float=1, Translation="0,0,7", Rotation="40,0,0", Scale="1,1,1") -> (Pose) {
    rotationMode=BMM_Additive
    rotationSpace=BCS_BoneSpace
    translationMode=BMM_Additive
    translationSpace=BCS_BoneSpace
    alphaBoolBlend {blendOption=Linear}
    boneToModify {boneName=dyn_hair_l_01}
    rotation {pitch=40}
    translation {z=7}
   }
   n72 = "Transform (Modify) Bone - Bone: dyn_hair_bk_01"!(ComponentPose=@n71, Alpha:float=1, Translation="0,0,6", Rotation="60,0,0", Scale="1,1,1") -> (Pose) {
    rotationMode=BMM_Additive
    rotationSpace=BCS_BoneSpace
    translationMode=BMM_Additive
    translationSpace=BCS_BoneSpace
    alphaBoolBlend {blendOption=Linear}
    boneToModify {boneName=dyn_hair_bk_01}
    rotation {pitch=60}
    translation {z=6}
   }
   n73 = "Transform (Modify) Bone - Bone: dyn_hair_r_01"!(ComponentPose=@n72, Alpha:float=1, Translation="-2,0,7", Rotation="60,0,0", Scale="1,1,1") -> (Pose) {
    rotationMode=BMM_Additive
    rotationSpace=BCS_BoneSpace
    translationMode=BMM_Additive
    translationSpace=BCS_BoneSpace
    alphaBoolBlend {blendOption=Linear}
    boneToModify {boneName=dyn_hair_r_01}
    rotation {pitch=60}
    translation {x=-2, z=7}
   }
   n74 = "Transform (Modify) Bone - Bone: dyn_hair_l_01"(ComponentPose=@n73, Alpha=1, Translation="0,0,0", Rotation="0,0,0", Scale="1,1,1") {
    rotationMode=BMM_Additive
    rotationSpace=BCS_BoneSpace
    translationMode=BMM_Additive
    translationSpace=BCS_BoneSpace
    alphaBoolBlend {blendOption=Linear}
    boneToModify {boneName=dyn_hair_l_01}
   }
   n75 = "Transform (Modify) Bone - Bone: dyn_hair_bk_01"(ComponentPose=@n74, Alpha=1, Translation="0,-9,0", Rotation="0,0,0", Scale="1,1,1") {
    rotationMode=BMM_Additive
    rotationSpace=BCS_BoneSpace
    translationMode=BMM_Additive
    translationSpace=BCS_BoneSpace
    alphaBoolBlend {blendOption=Linear}
    boneToModify {boneName=dyn_hair_bk_01}
    translation {y=-9}
   }
   n76 = "Transform (Modify) Bone - Bone: dyn_hair_r_01"(ComponentPose=@n75, Alpha=1, Translation="0,-4,-6", Rotation="0,0,0", Scale="1,1,1") {
    rotationMode=BMM_Additive
    rotationSpace=BCS_BoneSpace
    translationMode=BMM_Additive
    translationSpace=BCS_BoneSpace
    alphaBoolBlend {blendOption=Linear}
    boneToModify {boneName=dyn_hair_r_01}
    translation {y=-4, z=-6}
   }
   n77 = "Transform (Modify) Bone - Bone: dyn_hair_bk_01"(ComponentPose=@n76, Alpha=1, Translation="0,2,0", Rotation="0,0,0", Scale="1,1,1") {
    rotationMode=BMM_Additive
    rotationSpace=BCS_BoneSpace
    translationMode=BMM_Additive
    translationSpace=BCS_BoneSpace
    alphaBoolBlend {blendOption=Linear}
    boneToModify {boneName=dyn_hair_bk_01}
    translation {y=2}
   }
   n78 = "Transform (Modify) Bone - Bone: dyn_hair_l_01"(ComponentPose=@n77, Alpha=1, Translation="0,0,4", Rotation="0,0,-10", Scale="1,1,1") {
    rotationMode=BMM_Additive
    rotationSpace=BCS_BoneSpace
    translationMode=BMM_Additive
    translationSpace=BCS_BoneSpace
    alphaBoolBlend {blendOption=Linear}
    boneToModify {boneName=dyn_hair_l_01}
    rotation {roll=-10}
    translation {z=4}
   }
   n79 = "Transform (Modify) Bone - Bone: dyn_hair_bk_01"(ComponentPose=@n78, Alpha=1, Translation="0,-2,0", Rotation="0,0,0", Scale="1,1,1") {
    rotationMode=BMM_Additive
    rotationSpace=BCS_BoneSpace
    translationMode=BMM_Additive
    translationSpace=BCS_BoneSpace
    alphaBoolBlend {blendOption=Linear}
    boneToModify {boneName=dyn_hair_bk_01}
    translation {y=-2}
   }
   n80 = "Transform (Modify) Bone - Bone: dyn_hair_l_01"(ComponentPose=@n79, Alpha=1, Translation="0,0,0", Rotation="0,0,0", Scale="1,1,1") {
    rotationMode=BMM_Additive
    rotationSpace=BCS_BoneSpace
    translationMode=BMM_Additive
    translationSpace=BCS_BoneSpace
    alphaBoolBlend {blendOption=Linear}
    boneToModify {boneName=dyn_hair_l_01}
   }
   n81 = "Transform (Modify) Bone - Bone: dyn_hair_bk_01"(ComponentPose=@n80, Alpha=1, Translation="0,4,0", Rotation="0,0,0", Scale="1,1,1") {
    rotationMode=BMM_Additive
    rotationSpace=BCS_BoneSpace
    translationMode=BMM_Additive
    translationSpace=BCS_BoneSpace
    alphaBoolBlend {blendOption=Linear}
    boneToModify {boneName=dyn_hair_bk_01}
    translation {y=4}
   }
   n82 = "Transform (Modify) Bone - Bone: dyn_hair_r_01"(ComponentPose=@n81, Alpha=1, Translation="0,5,8", Rotation="0,0,40", Scale="1,1,1") {
    rotationMode=BMM_Additive
    rotationSpace=BCS_BoneSpace
    translationMode=BMM_Additive
    translationSpace=BCS_BoneSpace
    alphaBoolBlend {blendOption=Linear}
    boneToModify {boneName=dyn_hair_r_01}
    rotation {roll=40}
    translation {y=5, z=8}
   }
   n83 = "Transform (Modify) Bone - Bone: dyn_hair_r_01"(ComponentPose=@n82, Alpha=1, Translation="0,4,8", Rotation="0,0,90", Scale="1,1,1") {
    rotationMode=BMM_Additive
    rotationSpace=BCS_BoneSpace
    translationMode=BMM_Additive
    translationSpace=BCS_BoneSpace
    alphaBoolBlend {blendOption=Linear}
    boneToModify {boneName=dyn_hair_r_01}
    rotation {roll=90}
    translation {y=4, z=8}
   }
   n84 = "Transform (Modify) Bone - Bone: dyn_hair_bk_01"(ComponentPose=@n83, Alpha=1, Translation="0,4,0", Rotation="0,0,0", Scale="1,1,1") {
    rotationMode=BMM_Additive
    rotationSpace=BCS_BoneSpace
    translationMode=BMM_Additive
    translationSpace=BCS_BoneSpace
    alphaBoolBlend {blendOption=Linear}
    boneToModify {boneName=dyn_hair_bk_01}
    translation {y=4}
   }
   n85 = "Constraint - dyn_hair_bk_02 (P)"(ComponentPose=@n84, Alpha=1, ConstraintWeights_0=1) {
    alphaBoolBlend {blendOption=Linear}
    boneToModify {boneName=dyn_hair_bk_02}
    constraintSetup [{targetBone {boneName=spine_05}}]
    constraintWeights [1]
   }
   n86 = "Constraint - dyn_hair_l_02 (P)"(ComponentPose=@n85, Alpha=1, ConstraintWeights_0=1) {
    alphaBoolBlend {blendOption=Linear}
    boneToModify {boneName=dyn_hair_l_02}
    constraintSetup [{targetBone {boneName=spine_05}}]
    constraintWeights [1]
   }
   n87 = "Constraint - dyn_hair_r_02 (P)"(ComponentPose=@n86, Alpha=1, ConstraintWeights_0=1) {
    alphaBoolBlend {blendOption=Linear}
    boneToModify {boneName=dyn_hair_r_02}
    constraintSetup [{targetBone {boneName=spine_05}}]
    constraintWeights [1]
   }
   n88 = "Transform (Modify) Bone - Bone: dyn_hair_l_02"!(ComponentPose=@n87, Alpha:float=1, Translation="10,0,5", Rotation="0,0,0", Scale="1,1,1") -> (Pose) {
    rotationMode=BMM_Additive
    rotationSpace=BCS_BoneSpace
    translationMode=BMM_Additive
    translationSpace=BCS_BoneSpace
    alphaBoolBlend {blendOption=Linear}
    boneToModify {boneName=dyn_hair_l_02}
    translation {x=10, z=5}
   }
   n89 = "Transform (Modify) Bone - Bone: dyn_hair_bk_02"!(ComponentPose=@n88, Alpha:float=1, Translation="10,2,3", Rotation="0,0,0", Scale="1,1,1") -> (Pose) {
    rotationMode=BMM_Additive
    rotationSpace=BCS_BoneSpace
    translationMode=BMM_Additive
    translationSpace=BCS_BoneSpace
    alphaBoolBlend {blendOption=Linear}
    boneToModify {boneName=dyn_hair_bk_02}
    translation=<10 2 3>
   }
   n90 = "Transform (Modify) Bone - Bone: dyn_hair_r_02"!(ComponentPose=@n89, Alpha:float=1, Translation="10,3,5", Rotation="-20,0,0", Scale="1,1,1") -> (Pose) {
    rotationMode=BMM_Additive
    rotationSpace=BCS_BoneSpace
    translationMode=BMM_Additive
    translationSpace=BCS_BoneSpace
    alphaBoolBlend {blendOption=Linear}
    boneToModify {boneName=dyn_hair_r_02}
    rotation {pitch=-20}
    translation=<10 3 5>
   }
   n91 = "Transform (Modify) Bone - Bone: dyn_hair_l_02"(ComponentPose=@n90, Alpha=1, Translation="-1,0,0", Rotation="0,0,0", Scale="1,1,1") {
    rotationMode=BMM_Additive
    rotationSpace=BCS_BoneSpace
    translationMode=BMM_Additive
    translationSpace=BCS_BoneSpace
    alphaBoolBlend {blendOption=Linear}
    boneToModify {boneName=dyn_hair_l_02}
    translation {x=-1}
   }
   n92 = "Transform (Modify) Bone - Bone: dyn_hair_bk_02"(ComponentPose=@n91, Alpha=1, Translation="0,3,0", Rotation="0,0,0", Scale="1,1,1") {
    rotationMode=BMM_Additive
    rotationSpace=BCS_BoneSpace
    translationMode=BMM_Additive
    translationSpace=BCS_BoneSpace
    alphaBoolBlend {blendOption=Linear}
    boneToModify {boneName=dyn_hair_bk_02}
    translation {y=3}
   }
   n93 = "Transform (Modify) Bone - Bone: dyn_hair_r_02"(ComponentPose=@n92, Alpha=1, Translation="5,-3,5", Rotation="0,0,0", Scale="1,1,1") {
    rotationMode=BMM_Additive
    rotationSpace=BCS_BoneSpace
    translationMode=BMM_Additive
    translationSpace=BCS_BoneSpace
    alphaBoolBlend {blendOption=Linear}
    boneToModify {boneName=dyn_hair_r_02}
    translation=<5 -3 5>
   }
   n94 = "Transform (Modify) Bone - Bone: dyn_hair_bk_02"(ComponentPose=@n93, Alpha=1, Translation="0,3,1", Rotation="0,0,0", Scale="1,1,1") {
    rotationMode=BMM_Additive
    rotationSpace=BCS_BoneSpace
    translationMode=BMM_Additive
    translationSpace=BCS_BoneSpace
    alphaBoolBlend {blendOption=Linear}
    boneToModify {boneName=dyn_hair_bk_02}
    translation {y=3, z=1}
   }
   n95 = "Transform (Modify) Bone - Bone: dyn_hair_r_02"(ComponentPose=@n94, Alpha=1, Translation="2,0,5", Rotation="0,0,0", Scale="1,1,1") {
    rotationMode=BMM_Additive
    rotationSpace=BCS_BoneSpace
    translationMode=BMM_Additive
    translationSpace=BCS_BoneSpace
    alphaBoolBlend {blendOption=Linear}
    boneToModify {boneName=dyn_hair_r_02}
    translation {x=2, z=5}
   }
   n96 = "Transform (Modify) Bone - Bone: dyn_hair_l_02"(ComponentPose=@n95, Alpha=1, Translation="-3,-7,7", Rotation="0,-10,-20", Scale="1,1,1") {
    rotationMode=BMM_Additive
    rotationSpace=BCS_BoneSpace
    translationMode=BMM_Additive
    translationSpace=BCS_BoneSpace
    alphaBoolBlend {blendOption=Linear}
    boneToModify {boneName=dyn_hair_l_02}
    rotation {roll=-20, yaw=-10}
    translation=<-3 -7 7>
   }
   n97 = "Transform (Modify) Bone - Bone: dyn_hair_bk_02"(ComponentPose=@n96, Alpha=1, Translation="0,-7,0", Rotation="0,0,0", Scale="1,1,1") {
    rotationMode=BMM_Additive
    rotationSpace=BCS_BoneSpace
    translationMode=BMM_Additive
    translationSpace=BCS_BoneSpace
    alphaBoolBlend {blendOption=Linear}
    boneToModify {boneName=dyn_hair_bk_02}
    translation {y=-7}
   }
   n98 = "Transform (Modify) Bone - Bone: dyn_hair_l_02"(ComponentPose=@n97, Alpha=1, Translation="0,-1,3", Rotation="0,0,-30", Scale="1,1,1") {
    rotationMode=BMM_Additive
    rotationSpace=BCS_BoneSpace
    translationMode=BMM_Additive
    translationSpace=BCS_BoneSpace
    alphaBoolBlend {blendOption=Linear}
    boneToModify {boneName=dyn_hair_l_02}
    rotation {roll=-30}
    translation {y=-1, z=3}
   }
   n99 = "Transform (Modify) Bone - Bone: dyn_hair_bk_02"(ComponentPose=@n98, Alpha=1, Translation="0,-4,2", Rotation="0,0,0", Scale="1,1,1") {
    rotationMode=BMM_Additive
    rotationSpace=BCS_BoneSpace
    translationMode=BMM_Additive
    translationSpace=BCS_BoneSpace
    alphaBoolBlend {blendOption=Linear}
    boneToModify {boneName=dyn_hair_bk_02}
    translation {y=-4, z=2}
   }
   n100 = "Transform (Modify) Bone - Bone: dyn_hair_r_02"(ComponentPose=@n99, Alpha=1, Translation="3,-5,5", Rotation="0,0,40", Scale="1,1,1") {
    rotationMode=BMM_Additive
    rotationSpace=BCS_BoneSpace
    translationMode=BMM_Additive
    translationSpace=BCS_BoneSpace
    alphaBoolBlend {blendOption=Linear}
    boneToModify {boneName=dyn_hair_r_02}
    rotation {roll=40}
    translation=<3 -5 5>
   }
   n101 = "Transform (Modify) Bone - Bone: dyn_hair_r_02"(ComponentPose=@n100, Alpha=1, Translation="-4,2,3", Rotation="0,0,0", Scale="1,1,1") {
    rotationMode=BMM_Additive
    rotationSpace=BCS_BoneSpace
    translationMode=BMM_Additive
    translationSpace=BCS_BoneSpace
    alphaBoolBlend {blendOption=Linear}
    boneToModify {boneName=dyn_hair_r_02}
    translation=<-4 2 3>
   }
   n102 = "Transform (Modify) Bone - Bone: dyn_hair_bk_02"(ComponentPose=@n101, Alpha=1, Translation="0,0,0", Rotation="0,0,0", Scale="1,1,1") {
    rotationMode=BMM_Additive
    rotationSpace=BCS_BoneSpace
    translationMode=BMM_Additive
    translationSpace=BCS_BoneSpace
    alphaBoolBlend {blendOption=Linear}
    boneToModify {boneName=dyn_hair_bk_02}
   }
   n103 = "Transform (Modify) Bone - Bone: dyn_hair_r_02"(ComponentPose=@n102, Alpha=1, Translation="0,0,3", Rotation="0,0,0", Scale="1,1,1") {
    rotationMode=BMM_Additive
    rotationSpace=BCS_BoneSpace
    translationMode=BMM_Additive
    translationSpace=BCS_BoneSpace
    alphaBoolBlend {blendOption=Linear}
    boneToModify {boneName=dyn_hair_r_02}
    translation {z=3}
   }
   n104 = "Constraint - dyn_hair_bk_03 (P)"(ComponentPose=@n103, Alpha=1, ConstraintWeights_0=1) {
    alphaBoolBlend {blendOption=Linear}
    boneToModify {boneName=dyn_hair_bk_03}
    constraintSetup [{targetBone {boneName=spine_05}}]
    constraintWeights [1]
   }
   n105 = "Constraint - dyn_hair_l_03 (P)"(ComponentPose=@n104, Alpha=1, ConstraintWeights_0=1) {
    alphaBoolBlend {blendOption=Linear}
    boneToModify {boneName=dyn_hair_l_03}
    constraintSetup [{targetBone {boneName=spine_05}}]
    constraintWeights [1]
   }
   n106 = "Constraint - dyn_hair_r_03 (P)"(ComponentPose=@n105, Alpha=1, ConstraintWeights_0=1) {
    alphaBoolBlend {blendOption=Linear}
    boneToModify {boneName=dyn_hair_r_03}
    constraintSetup [{targetBone {boneName=spine_05}}]
    constraintWeights [1]
   }
   n107 = "Transform (Modify) Bone - Bone: dyn_hair_l_03"!(ComponentPose=@n106, Alpha:float=1, Translation="8,-3,2", Rotation="-20,0,0", Scale="1,1,1") -> (Pose) {
    rotationMode=BMM_Additive
    rotationSpace=BCS_BoneSpace
    translationMode=BMM_Additive
    translationSpace=BCS_BoneSpace
    alphaBoolBlend {blendOption=Linear}
    boneToModify {boneName=dyn_hair_l_03}
    rotation {pitch=-20}
    translation=<8 -3 2>
   }
   n108 = "Transform (Modify) Bone - Bone: dyn_hair_bk_03"!(ComponentPose=@n107, Alpha:float=1, Translation="10,0,-1", Rotation="-10,0,0", Scale="1,1,1") -> (Pose) {
    rotationMode=BMM_Additive
    rotationSpace=BCS_BoneSpace
    translationMode=BMM_Additive
    translationSpace=BCS_BoneSpace
    alphaBoolBlend {blendOption=Linear}
    boneToModify {boneName=dyn_hair_bk_03}
    rotation {pitch=-10}
    translation {x=10, z=-1}
   }
   n109 = "Transform (Modify) Bone - Bone: dyn_hair_r_03"!(ComponentPose=@n108, Alpha:float=1, Translation="8,2,-1", Rotation="-10,0,0", Scale="1,1,1") -> (Pose) {
    rotationMode=BMM_Additive
    rotationSpace=BCS_BoneSpace
    translationMode=BMM_Additive
    translationSpace=BCS_BoneSpace
    alphaBoolBlend {blendOption=Linear}
    boneToModify {boneName=dyn_hair_r_03}
    rotation {pitch=-10}
    translation=<8 2 -1>
   }
   n110 = "Transform (Modify) Bone - Bone: dyn_hair_l_03"(ComponentPose=@n109, Alpha=1, Translation="-2,-1,2", Rotation="0,0,0", Scale="1,1,1") {
    rotationMode=BMM_Additive
    rotationSpace=BCS_BoneSpace
    translationMode=BMM_Additive
    translationSpace=BCS_BoneSpace
    alphaBoolBlend {blendOption=Linear}
    boneToModify {boneName=dyn_hair_l_03}
    translation=<-2 -1 2>
   }
   n111 = "Transform (Modify) Bone - Bone: dyn_hair_bk_03"(ComponentPose=@n110, Alpha=1, Translation="0,-1,0", Rotation="0,0,0", Scale="1,1,1") {
    rotationMode=BMM_Additive
    rotationSpace=BCS_BoneSpace
    translationMode=BMM_Additive
    translationSpace=BCS_BoneSpace
    alphaBoolBlend {blendOption=Linear}
    boneToModify {boneName=dyn_hair_bk_03}
    translation {y=-1}
   }
   n112 = "Transform (Modify) Bone - Bone: dyn_hair_r_03"(ComponentPose=@n111, Alpha=1, Translation="0,0,0", Rotation="0,0,0", Scale="1,1,1") {
    rotationMode=BMM_Additive
    rotationSpace=BCS_BoneSpace
    translationMode=BMM_Additive
    translationSpace=BCS_BoneSpace
    alphaBoolBlend {blendOption=Linear}
    boneToModify {boneName=dyn_hair_r_03}
   }
   n113 = "Transform (Modify) Bone - Bone: dyn_hair_bk_03"(ComponentPose=@n112, Alpha=1, Translation="0,1,2", Rotation="0,0,0", Scale="1,1,1") {
    rotationMode=BMM_Additive
    rotationSpace=BCS_BoneSpace
    translationMode=BMM_Additive
    translationSpace=BCS_BoneSpace
    alphaBoolBlend {blendOption=Linear}
    boneToModify {boneName=dyn_hair_bk_03}
    translation {y=1, z=2}
   }
   n114 = "Transform (Modify) Bone - Bone: dyn_hair_r_03"(ComponentPose=@n113, Alpha=1, Translation="-5,3,4", Rotation="0,0,0", Scale="1,1,1") {
    rotationMode=BMM_Additive
    rotationSpace=BCS_BoneSpace
    translationMode=BMM_Additive
    translationSpace=BCS_BoneSpace
    alphaBoolBlend {blendOption=Linear}
    boneToModify {boneName=dyn_hair_r_03}
    translation=<-5 3 4>
   }
   n115 = "Transform (Modify) Bone - Bone: dyn_hair_l_03"(ComponentPose=@n114, Alpha=1, Translation="-5,-6,9", Rotation="0,-10,-30", Scale="1,1,1") {
    rotationMode=BMM_Additive
    rotationSpace=BCS_BoneSpace
    translationMode=BMM_Additive
    translationSpace=BCS_BoneSpace
    alphaBoolBlend {blendOption=Linear}
    boneToModify {boneName=dyn_hair_l_03}
    rotation {roll=-30, yaw=-10}
    translation=<-5 -6 9>
   }
   n116 = "Transform (Modify) Bone - Bone: dyn_hair_bk_03"(ComponentPose=@n115, Alpha=1, Translation="0,-2,0", Rotation="0,0,0", Scale="1,1,1") {
    rotationMode=BMM_Additive
    rotationSpace=BCS_BoneSpace
    translationMode=BMM_Additive
    translationSpace=BCS_BoneSpace
    alphaBoolBlend {blendOption=Linear}
    boneToModify {boneName=dyn_hair_bk_03}
    translation {y=-2}
   }
   n117 = "Transform (Modify) Bone - Bone: dyn_hair_l_03"(ComponentPose=@n116, Alpha=1, Translation="-3,-2,4", Rotation="0,0,-20", Scale="1,1,1") {
    rotationMode=BMM_Additive
    rotationSpace=BCS_BoneSpace
    translationMode=BMM_Additive
    translationSpace=BCS_BoneSpace
    alphaBoolBlend {blendOption=Linear}
    boneToModify {boneName=dyn_hair_l_03}
    rotation {roll=-20}
    translation=<-3 -2 4>
   }
   n118 = "Transform (Modify) Bone - Bone: dyn_hair_bk_03"(ComponentPose=@n117, Alpha=1, Translation="-2,-2,0", Rotation="0,0,0", Scale="1,1,1") {
    rotationMode=BMM_Additive
    rotationSpace=BCS_BoneSpace
    translationMode=BMM_Additive
    translationSpace=BCS_BoneSpace
    alphaBoolBlend {blendOption=Linear}
    boneToModify {boneName=dyn_hair_bk_03}
    translation=<-2 -2>
   }
   n119 = "Transform (Modify) Bone - Bone: dyn_hair_r_03"(ComponentPose=@n118, Alpha=1, Translation="0,0,2", Rotation="0,0,0", Scale="1,1,1") {
    rotationMode=BMM_Additive
    rotationSpace=BCS_BoneSpace
    translationMode=BMM_Additive
    translationSpace=BCS_BoneSpace
    alphaBoolBlend {blendOption=Linear}
    boneToModify {boneName=dyn_hair_r_03}
    translation {z=2}
   }
   n120 = "Transform (Modify) Bone - Bone: dyn_hair_r_03"(ComponentPose=@n119, Alpha=1, Translation="0,0,0", Rotation="0,0,0", Scale="1,1,1") {
    rotationMode=BMM_Additive
    rotationSpace=BCS_BoneSpace
    translationMode=BMM_Additive
    translationSpace=BCS_BoneSpace
    alphaBoolBlend {blendOption=Linear}
    boneToModify {boneName=dyn_hair_r_03}
   }
   n121 = "Transform (Modify) Bone - Bone: dyn_hair_bk_03"(ComponentPose=@n120, Alpha=1, Translation="0,0,0", Rotation="0,0,0", Scale="1,1,1") {
    rotationMode=BMM_Additive
    rotationSpace=BCS_BoneSpace
    translationMode=BMM_Additive
    translationSpace=BCS_BoneSpace
    alphaBoolBlend {blendOption=Linear}
    boneToModify {boneName=dyn_hair_bk_03}
   }
   n122 = "Transform (Modify) Bone - Bone: dyn_hair_r_03"(ComponentPose=@n121, Alpha=1, Translation="0,2,3", Rotation="0,0,30", Scale="1,1,1") {
    rotationMode=BMM_Additive
    rotationSpace=BCS_BoneSpace
    translationMode=BMM_Additive
    translationSpace=BCS_BoneSpace
    alphaBoolBlend {blendOption=Linear}
    boneToModify {boneName=dyn_hair_r_03}
    rotation {roll=30}
    translation {y=2, z=3}
   }
   n123 = "Transform (Modify) Bone - Bone: dyn_hair_l_03"(ComponentPose=@n122, Alpha=1, Translation="0,0,0", Rotation="40,-20,0", Scale="1,1,1") {
    rotationMode=BMM_Additive
    rotationSpace=BCS_BoneSpace
    translationMode=BMM_Additive
    translationSpace=BCS_BoneSpace
    alphaBoolBlend {blendOption=Linear}
    boneToModify {boneName=dyn_hair_l_03}
    rotation {pitch=40, yaw=-20}
   }
   n124 = "Transform (Modify) Bone - Bone: dyn_hair_r_03"(ComponentPose=@n123, Alpha=1, Translation="0,0,0", Rotation="10,0,0", Scale="1,1,1") {
    rotationMode=BMM_Additive
    rotationSpace=BCS_BoneSpace
    translationMode=BMM_Additive
    translationSpace=BCS_BoneSpace
    alphaBoolBlend {blendOption=Linear}
    boneToModify {boneName=dyn_hair_r_03}
    rotation {pitch=10}
   }
   n125 = "Constraint - dyn_hair_bk_04 (P)"(ComponentPose=@n124, Alpha=1, ConstraintWeights_0=1) {
    alphaBoolBlend {blendOption=Linear}
    boneToModify {boneName=dyn_hair_bk_04}
    constraintSetup [{targetBone {boneName=spine_05}}]
    constraintWeights [1]
   }
   n126 = "Constraint - dyn_hair_l_04 (P)"(ComponentPose=@n125, Alpha=1, ConstraintWeights_0=1) {
    alphaBoolBlend {blendOption=Linear}
    boneToModify {boneName=dyn_hair_l_04}
    constraintSetup [{targetBone {boneName=spine_05}}]
    constraintWeights [1]
   }
   n127 = "Constraint - dyn_hair_r_04 (P)"(ComponentPose=@n126, Alpha=1, ConstraintWeights_0=1) {
    alphaBoolBlend {blendOption=Linear}
    boneToModify {boneName=dyn_hair_r_04}
    constraintSetup [{targetBone {boneName=spine_05}}]
    constraintWeights [1]
   }
   n128 = "Transform (Modify) Bone - Bone: dyn_hair_l_04"!(ComponentPose=@n127, Alpha:float=1, Translation="8,-2,-1", Rotation="0,0,0", Scale="1,1,1") -> (Pose) {
    rotationMode=BMM_Additive
    rotationSpace=BCS_BoneSpace
    translationMode=BMM_Additive
    translationSpace=BCS_BoneSpace
    alphaBoolBlend {blendOption=Linear}
    boneToModify {boneName=dyn_hair_l_04}
    translation=<8 -2 -1>
   }
   n129 = "Transform (Modify) Bone - Bone: dyn_hair_bk_04"!(ComponentPose=@n128, Alpha:float=1, Translation="8,0,-1", Rotation="-10,0,0", Scale="1,1,1") -> (Pose) {
    rotationMode=BMM_Additive
    rotationSpace=BCS_BoneSpace
    translationMode=BMM_Additive
    translationSpace=BCS_BoneSpace
    alphaBoolBlend {blendOption=Linear}
    boneToModify {boneName=dyn_hair_bk_04}
    rotation {pitch=-10}
    translation {x=8, z=-1}
   }
   n130 = "Transform (Modify) Bone - Bone: dyn_hair_r_04"!(ComponentPose=@n129, Alpha:float=1, Translation="8,2,-1", Rotation="-10,0,0", Scale="1,1,1") -> (Pose) {
    rotationMode=BMM_Additive
    rotationSpace=BCS_BoneSpace
    translationMode=BMM_Additive
    translationSpace=BCS_BoneSpace
    alphaBoolBlend {blendOption=Linear}
    boneToModify {boneName=dyn_hair_r_04}
    rotation {pitch=-10}
    translation=<8 2 -1>
   }
   n131 = "Transform (Modify) Bone - Bone: dyn_hair_l_04"(ComponentPose=@n130, Alpha=1, Translation="-3,-3,0", Rotation="0,0,0", Scale="1,1,1") {
    rotationMode=BMM_Additive
    rotationSpace=BCS_BoneSpace
    translationMode=BMM_Additive
    translationSpace=BCS_BoneSpace
    alphaBoolBlend {blendOption=Linear}
    boneToModify {boneName=dyn_hair_l_04}
    translation=<-3 -3>
   }
   n132 = "Transform (Modify) Bone - Bone: dyn_hair_bk_04"(ComponentPose=@n131, Alpha=1, Translation="0,-2,0", Rotation="0,0,0", Scale="1,1,1") {
    rotationMode=BMM_Additive
    rotationSpace=BCS_BoneSpace
    translationMode=BMM_Additive
    translationSpace=BCS_BoneSpace
    alphaBoolBlend {blendOption=Linear}
    boneToModify {boneName=dyn_hair_bk_04}
    translation {y=-2}
   }
   n133 = "Transform (Modify) Bone - Bone: dyn_hair_r_04"(ComponentPose=@n132, Alpha=1, Translation="0,0,0", Rotation="0,0,0", Scale="1,1,1") {
    rotationMode=BMM_Additive
    rotationSpace=BCS_BoneSpace
    translationMode=BMM_Additive
    translationSpace=BCS_BoneSpace
    alphaBoolBlend {blendOption=Linear}
    boneToModify {boneName=dyn_hair_r_04}
   }
   n134 = "Transform (Modify) Bone - Bone: dyn_hair_r_04"(ComponentPose=@n133, Alpha=1, Translation="-5,3,3", Rotation="0,0,0", Scale="1,1,1") {
    rotationMode=BMM_Additive
    rotationSpace=BCS_BoneSpace
    translationMode=BMM_Additive
    translationSpace=BCS_BoneSpace
    alphaBoolBlend {blendOption=Linear}
    boneToModify {boneName=dyn_hair_r_04}
    translation=<-5 3 3>
   }
   n135 = "Transform (Modify) Bone - Bone: dyn_hair_l_04"(ComponentPose=@n134, Alpha=1, Translation="-5,-5,6", Rotation="0,0,-30", Scale="1,1,1") {
    rotationMode=BMM_Additive
    rotationSpace=BCS_BoneSpace
    translationMode=BMM_Additive
    translationSpace=BCS_BoneSpace
    alphaBoolBlend {blendOption=Linear}
    boneToModify {boneName=dyn_hair_l_04}
    rotation {roll=-30}
    translation=<-5 -5 6>
   }
   n136 = "Transform (Modify) Bone - Bone: dyn_hair_bk_04"(ComponentPose=@n135, Alpha=1, Translation="0,-1,0", Rotation="0,0,0", Scale="1,1,1") {
    rotationMode=BMM_Additive
    rotationSpace=BCS_BoneSpace
    translationMode=BMM_Additive
    translationSpace=BCS_BoneSpace
    alphaBoolBlend {blendOption=Linear}
    boneToModify {boneName=dyn_hair_bk_04}
    translation {y=-1}
   }
   n137 = "Transform (Modify) Bone - Bone: dyn_hair_l_04"(ComponentPose=@n136, Alpha=1, Translation="-3,-3,2", Rotation="0,0,0", Scale="1,1,1") {
    rotationMode=BMM_Additive
    rotationSpace=BCS_BoneSpace
    translationMode=BMM_Additive
    translationSpace=BCS_BoneSpace
    alphaBoolBlend {blendOption=Linear}
    boneToModify {boneName=dyn_hair_l_04}
    translation=<-3 -3 2>
   }
   n138 = "Transform (Modify) Bone - Bone: dyn_hair_bk_04"(ComponentPose=@n137, Alpha=1, Translation="0,-3,0", Rotation="0,0,0", Scale="1,1,1") {
    rotationMode=BMM_Additive
    rotationSpace=BCS_BoneSpace
    translationMode=BMM_Additive
    translationSpace=BCS_BoneSpace
    alphaBoolBlend {blendOption=Linear}
    boneToModify {boneName=dyn_hair_bk_04}
    translation {y=-3}
   }
   n139 = "Transform (Modify) Bone - Bone: dyn_hair_r_04"(ComponentPose=@n138, Alpha=1, Translation="0,0,-1", Rotation="-5,0,0", Scale="1,1,1") {
    rotationMode=BMM_Additive
    rotationSpace=BCS_BoneSpace
    translationMode=BMM_Additive
    translationSpace=BCS_BoneSpace
    alphaBoolBlend {blendOption=Linear}
    boneToModify {boneName=dyn_hair_r_04}
    rotation {pitch=-5}
    translation {z=-1}
   }
   n140 = "Transform (Modify) Bone - Bone: dyn_hair_r_04"(ComponentPose=@n139, Alpha=1, Translation="0,0,0", Rotation="0,0,0", Scale="1,1,1") {
    rotationMode=BMM_Additive
    rotationSpace=BCS_BoneSpace
    translationMode=BMM_Additive
    translationSpace=BCS_BoneSpace
    alphaBoolBlend {blendOption=Linear}
    boneToModify {boneName=dyn_hair_r_04}
   }
   n141 = "Transform (Modify) Bone - Bone: dyn_hair_r_04"(ComponentPose=@n140, Alpha=1, Translation="0,2,3", Rotation="0,0,30", Scale="1,1,1") {
    rotationMode=BMM_Additive
    rotationSpace=BCS_BoneSpace
    translationMode=BMM_Additive
    translationSpace=BCS_BoneSpace
    alphaBoolBlend {blendOption=Linear}
    boneToModify {boneName=dyn_hair_r_04}
    rotation {roll=30}
    translation {y=2, z=3}
   }
   n142 = "Transform (Modify) Bone - Bone: dyn_hair_l_04"(ComponentPose=@n141, Alpha=1, Translation="0,-1,0", Rotation="10,0,0", Scale="1,1,1") {
    rotationMode=BMM_Additive
    rotationSpace=BCS_BoneSpace
    translationMode=BMM_Additive
    translationSpace=BCS_BoneSpace
    alphaBoolBlend {blendOption=Linear}
    boneToModify {boneName=dyn_hair_l_04}
    rotation {pitch=10}
    translation {y=-1}
   }
   n143 = "Transform (Modify) Bone - Bone: dyn_hair_r_04"(ComponentPose=@n142, Alpha=1, Translation="0,1,0", Rotation="25,0,20", Scale="1,1,1") {
    rotationMode=BMM_Additive
    rotationSpace=BCS_BoneSpace
    translationMode=BMM_Additive
    translationSpace=BCS_BoneSpace
    alphaBoolBlend {blendOption=Linear}
    boneToModify {boneName=dyn_hair_r_04}
    rotation {pitch=25, roll=20}
    translation {y=1}
   }
   n144 = "Constraint - dyn_hair_bk_05 (P)"(ComponentPose=@n143, Alpha=1, ConstraintWeights_0=1) {
    alphaBoolBlend {blendOption=Linear}
    boneToModify {boneName=dyn_hair_bk_05}
    constraintSetup [{targetBone {boneName=dyn_hair_bk_04}}]
    constraintWeights [1]
   }
   n145 = "Constraint - dyn_hair_l_05 (P)"(ComponentPose=@n144, Alpha=1, ConstraintWeights_0=1) {
    alphaBoolBlend {blendOption=Linear}
    boneToModify {boneName=dyn_hair_l_05}
    constraintSetup [{targetBone {boneName=dyn_hair_l_04}}]
    constraintWeights [1]
   }
   n146 = "Constraint - dyn_hair_r_05 (P)"(ComponentPose=@n145, Alpha=1, ConstraintWeights_0=1) {
    alphaBoolBlend {blendOption=Linear}
    boneToModify {boneName=dyn_hair_r_05}
    constraintSetup [{targetBone {boneName=dyn_hair_r_04}}]
    constraintWeights [1]
   }
   n147 = "Transform (Modify) Bone - Bone: dyn_hair_l_05"(ComponentPose=@n146, Alpha=1, Translation="0,0,0", Rotation="0,0,0", Scale="1,1,1") {
    rotationMode=BMM_Additive
    rotationSpace=BCS_BoneSpace
    translationMode=BMM_Additive
    translationSpace=BCS_BoneSpace
    alphaBoolBlend {blendOption=Linear}
    boneToModify {boneName=dyn_hair_l_05}
   }
   n148 = "Transform (Modify) Bone - Bone: dyn_hair_bk_05"(ComponentPose=@n147, Alpha=1, Translation="0,0,1", Rotation="40,0,0", Scale="1,1,1") {
    rotationMode=BMM_Additive
    rotationSpace=BCS_BoneSpace
    translationMode=BMM_Additive
    translationSpace=BCS_BoneSpace
    alphaBoolBlend {blendOption=Linear}
    boneToModify {boneName=dyn_hair_bk_05}
    rotation {pitch=40}
    translation {z=1}
   }
   n149 = "Transform (Modify) Bone - Bone: dyn_hair_r_05"!(ComponentPose=@n148, Alpha:float=1, Translation="0,6,0", Rotation="20,0,0", Scale="1,1,1") -> (Pose) {
    rotationMode=BMM_Additive
    rotationSpace=BCS_BoneSpace
    translationMode=BMM_Additive
    translationSpace=BCS_BoneSpace
    alphaBoolBlend {blendOption=Linear}
    boneToModify {boneName=dyn_hair_r_05}
    rotation {pitch=20}
    translation {y=6}
   }
   Save cached pose 'PreSimPose'(Pose=Component To Local(ComponentPose=@n149)) {cacheName=PreSimPose}
  ]}
 EventGraph {nodes [
   Event BlueprintInitializeAnimation()
   > Set SkelMesh(SkelMesh=GetOwnerSkeletalMeshComponent())
   n1 = CheckForBackbling()
   > n2 = Set WearingBackpack(WearingBackpack:bool=AND Boolean(A=@n1, B=NOT Boolean(A=@n1.bIsBackblingACape))) -> (Output_Get:bool)
   > Set NoBackpack(NoBackpack:bool=NOT Boolean(A=@n2)) -> (Output_Get:bool)
   Event OnCharacterCustomizationLoaded()
   > Set SkelMesh(SkelMesh=GetPartSkeletalMeshComponent())
   Event BlueprintUpdateAnimation()
   > n3 = Sequence!(execute) -> (then_0, then_1, then_2:Exec)
   n4 = GetOwnerSkeletalMeshComponent()
   n5 = BreakRotator(InRot=BreakTransform(InTransform=GetDeltaTransformFromRefPose(self=@n4, BoneName=head, BaseName=spine_05)).Rotation)
   > Set HeadRorR(HeadRorR:double=MapRangeClamped(Value=@n5.Pitch, InRangeA=20, InRangeB=53)) -> (Output_Get:double)
   > Set HeadRorBk(HeadRorBk:double=MapRangeClamped(Value=@n5.Yaw, InRangeA=-10, InRangeB=-55)) -> (Output_Get:double)
   > Set HeadRotLeft(HeadRotLeft:double=MapRangeClamped(Value=@n5, InRangeA=15, InRangeB=65)) -> (Output_Get:double)
   > Set HeadRotRight(HeadRotRight:double=MapRangeClamped(Value=@n5, InRangeA=-20, InRangeB=-65)) -> (Output_Get:double)
   n6 = NOT Boolean(A=Get bIsParachuteOpened())
   n7 = Get bIsSkyDiving()
   n8 = AND Boolean!(A=@n6, B=NOT Boolean(A=Get bIsSkydiveDiveMode()), C:bool=@n7) -> (ReturnValue)
   > Set IsSkyJump(IsSkyJump:bool=AND Boolean!(A=@n8, B=Get bIsJumping()) -> (ReturnValue)) -> (Output_Get:bool)
   > Set IsSkyFall(IsSkyFall:bool=AND Boolean(A=@n8, B=Get bIsFalling())) -> (Output_Get:bool)
   > Set IsInSkyWithoutParachute(IsInSkyWithoutParachute:bool=AND Boolean(A=@n6, B=@n7)) -> (Output_Get:bool)
   n9 = GetOwnerSkeletalMeshComponent()
   n10 = BreakRotator(InRot=BreakTransform(InTransform=GetDeltaTransformFromRefPose(self=@n9, BoneName=clavicle_l, BaseName=spine_05)).Rotation)
   Set Clavicle_L_RotTop(execute=@n3.then_1, Clavicle_L_RotTop:double=float - float(A=MapRangeClamped(Value=@n10.Pitch, InRangeA=10, InRangeB=40), B=Get HeadRorBk())) -> (Output_Get:double)
   > Set Clavicle_L_RotBk(Clavicle_L_RotBk:double=float - float(A=MapRangeClamped(Value=@n10.Yaw, InRangeA=-10, InRangeB=-25), B=Get HeadRorBk())) -> (Output_Get:double)
   n11 = BreakRotator(InRot=BreakTransform(InTransform=GetDeltaTransformFromRefPose(self=@n9, BoneName=clavicle_r, BaseName=spine_05)).Rotation)
   > Set Clavicle_R_RotTop(Clavicle_R_RotTop:double=float - float!(A=MapRangeClamped(Value=@n11.Pitch, InRangeA=0, InRangeB=35), B=Get HeadRorR(), C:double=Get HeadRorBk()) -> (ReturnValue)) -> (Output_Get:double)
   > Set Clavicle_R_RotBk(Clavicle_R_RotBk:double=MapRangeClamped(Value=@n11.Yaw, InRangeA=0, InRangeB=-20)) -> (Output_Get:double)
   n12 = BreakRotator(InRot=BreakTransform(InTransform=GetDeltaTransformFromRefPose(self=@n9, BoneName=upperarm_r, BaseName=clavicle_r)).Rotation)
   > Set Arm_R_RotBk(Arm_R_RotBk:double=float - float(A=MapRangeClamped(Value=@n12.Yaw, InRangeA=-10, InRangeB=-30), B:bool=Get bIsSkyDiving(), C=Get HeadRotLeft() -> (HeadRotLeft:double))) -> (Output_Get:double)
   n13 = BreakRotator(InRot=BreakTransform(InTransform=GetDeltaTransformFromRefPose(self=@n9, BoneName=upperarm_l, BaseName=clavicle_l)).Rotation)
   > Set Arm_L_RotBk(Arm_L_RotBk:double=float - float!(A=MapRangeClamped(Value=@n13.Yaw, InRangeA=0, InRangeB=-35), B=Get bIsSkyDiving()) -> (ReturnValue)) -> (Output_Get:double)
   n14 = Gravity Override State Function DEPRECATED(
    execute=@n3.then_2
    CharacterPartAnimInstance=Self-Reference()
    GameTimeInSeconds=GetGameTimeInSeconds()
    GravityOverrideParams=Get Gravity Override Params()
   )
   > Set GravityOverride(GravityOverride=@n14)
   > n15 = Branch(Condition=Get bIsSkyDiving())
   > Set AccScaleOverride(AccScaleOverride:Vector="0.01,0.01,0.01") -> (Output_Get:Vector)
   Set AccScaleOverride(execute=@n15.else, AccScaleOverride="0.28,0.28,0.35")
   MapRangeClamped(Value=@n12.Pitch, InRangeA=5, InRangeB=-15)
  ]}
]
variables [
 AccScaleOverride {type=Vector, default=<0.6 0.6 0.2>}
 Arm_L_RotBk {type=number}
 Arm_R_RotBk {type=number}
 Clavicle_L_RotBk {type=number}
 Clavicle_L_RotTop {type=number}
 Clavicle_R_RotBk {type=number}
 Clavicle_R_RotTop {type=number}
 Gravity Override Params {
  type=GravityOverrideParamsStruct
  default {
   dBNO {gravityOverride=<-700 200 0>}
   dBNO_Carried {gravityOverride=<-700 300 -300>}
   emoteOrMelee {gravityOverride=<0 0 -700>}
   front_End {gravityOverride=<0 0 -700>}
   onGround_Falling {gravityOverride=<0 0 -700>}
   onGround_Moving {gravityOverride=<0 0 -700>}
   onGround_Standing {gravityOverride=<0 0 -700>}
   skydive_DivingDown {gravityOverride=<-1500 600 0>, windAmplitude=<50 50 100>}
   skydive_FloatingDown {gravityOverride=<0 -100 -350>, windAmplitude=<50 -100 50>, windOffset=<0 0 50>}
   skydive_GoingUp {gravityOverride=<0 -100 -300>, windAmplitude=<50 -200 100>}
   skydive_Parachuting {gravityOverride=<0 -500 -700>, windAmplitude=<100 -150 200>}
   swimming {gravityOverride=<-500 200 0>}
   ziplining {gravityOverride=<0 0 -700>}
  }
 }
 GravityOverride {type=Vector}
 HeadRorBk {type=number}
 HeadRorR {type=number}
 HeadRotLeft {type=number}
 HeadRotRight {type=number}
 IsInSkyWithoutParachute {type=boolean}
 IsSkyFall {type=boolean}
 IsSkyJump {type=boolean}
 NoBackpack {type=boolean}
 SkelMesh {type=SkeletalMeshComponent}
 WearingBackpack {type=boolean}
]
